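# Niagara system for the 2D Smoke vehicle boost cosmetic, a stylized exhaust trail built from five sprite emitters: Bulk, Tiny, Rings, Pointed and Source.
# Each emitter partitions particles across filtered skeletal mesh sockets, applies per-layer painted color variants and fires them backward with SubUV animation.
# All but Source take velocity from game mode and inherit decaying source movement; Source adds short-lived cone bursts with wind.
# A frontend-preview flag retunes lifetime, spawning and drag.
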NiagaraSystem /VehicleCosmetics/Boosts/Boost_2D_Smoke/FX/NS_Boost_2D_Smoke
dependencies {
 usedDataInterfaces [
  &NiagaraDataInterfaceCurve
  &NiagaraDataInterfaceVector2DCurve
  &NiagaraDataInterfaceSpriteRendererInfo
  &NiagaraDataInterfaceSkeletalMesh
 ]
 usedDynamicInputs []
 usedModules [
  &self:SystemSpawnScript.SystemScriptSource.SystemScriptGraph.NiagaraNodeAssignment_1.SetVariables_#1
  &SystemState
 ]
}
emitters {
 NE_VC_Boost_Bulk {
  emitterUpdateScript [
   &EmitterState(Life Cycle Mode=::NewEnumerator1, Loop Duration=0.1, Recalculate Duration Each Loop=true, UseLoopDelay=false)
   &VC_FilteredSocketCount()
   &SpawnRate(
    SpawnRate=&MakeCustomFloatFromBool(Bool=$User.bIsFrontendPreview, True Float=&RandomRangeFloat(Minimum=160, Maximum=300), False Float=250)
    Use Spawn Probability=false
   )
  ]
  particleSpawnScript [
   &VC_PaintedVariantsOverride(
    Black=<rgba 0.003 0.003 0.003 1>
    BurntSienna=<rgba 0.005 0.003 0.002 1>
    Cobalt=<rgba 0.006 0.0088 0.04 1>
    Crimson=<rgba 0.02 0.0021 0.0026 1>
    DefaultColor=<rgba 0.0047 0.2135 0.1961 1>
    ForestGreen=<rgba 0.0027 0.013 0.0032 1>
    Grey=<rgba 0.008 0.008 0.008 1>
    Lime=<rgba 0.015 0.03 0 1>
    Orange=<rgba 0.02 0.0063 0 1>
    Pink=<rgba 0.03 0.0111 0.0268 1>
    Purple=<rgba 0.01 0 0.05 1>
    Saffron=<rgba 0.0276 0.03 0.0015 1>
    SkyBlue=<rgba 0 0.03 0.04 1>
    TitaniumWhite=<rgba 0.04 0.04 0.04 1>
   )
   &InitializeParticle(
    Color Mode=::NewEnumerator0
    Lifetime=&MakeCustomFloatFromBool(Bool=$User.bIsFrontendPreview, True Float=1.1, False Float=0.5)
    Sprite Rotation Mode=::NewEnumerator2
    Sprite Size Mode=::NewEnumerator1
    Uniform Sprite Size Max=50
    Uniform Sprite Size Min=15
    UsePositionOffset=false
    Write Material Random=true
   )
   &PartitionParticles(Number Of Partitions=$Emitter.PortCount)
   &SkeletalMeshLocation(
    Bone / Socket Index=$Particles.PartitionParticles.Partition
    Bone / Socket Sampling Mode=::NewEnumerator5
    MaskSpawnGroup=false
    Mesh Sampling Type=::NewEnumerator6
    Sampled Position Offset=&AddVectorToPosition(Position=<0 0 0>, Vector=$User.DTCarOffset)
    Skeletal Mesh=$System.SkeletalMesh
   )
   &AddVelocity(
    Velocity=&MergeHelper_VectorSwitch(0=<1000 0 0>, 1=<1000 0 0>, 2=<100 0 0>, 3=?, Index=$User.GameModeID, Num=::NewEnumerator1)
    Velocity Speed Scale=-1
   )
   &CameraOffset(Camera Offset Amount=-50)
  ]
  particleUpdateScript [
   &ParticleState()
   &ScaleSpriteSize(Uniform Curve Sprite Scale=DataInterface() {propertyValues {Curve {keys [
        {arriveTangent=3.8917, interpMode=RCIM_Cubic, leaveTangent=3.8917, tangentMode=RCTM_User, time=0, value=0.5}
        {arriveTangent=1.6712, interpMode=RCIM_Cubic, leaveTangent=1.6712, tangentMode=RCTM_User, time=0.2, value=1.1084}
        {arriveTangent=0.0224, interpMode=RCIM_Cubic, leaveTangent=0.0224, tangentMode=RCTM_User, time=1, value=1.5}
       ]}}})
   &SpriteRotationRate(Delta Time:NiagaraFloat=$Particles.Lifetime, Rotation Rate=&RandomRangeFloat(Minimum=10, Maximum=-10))
   &InheritSourceMovement(
    Applied Position Scale=&Multiply_VectorByFloat(
     Vector=<1 1 1>
     Float=&FloatFromCurve(FloatCurve=DataInterface() {propertyValues {Curve {keys [
          {arriveTangent=0.0056, interpMode=RCIM_Cubic, leaveTangent=0.0056, tangentMode=RCTM_User, time=0.1, value=1}
          {arriveTangent=-0.248, interpMode=RCIM_Cubic, leaveTangent=-0.248, tangentMode=RCTM_User, time=1, value=0.5}
         ]}}})
    )
    ApplyForce=false
    ApplyPosition=true
    ApplyPositionFalloff=false
   )
   &SolveForcesAndVelocity(Clamp Velocity=false, Limit Acceleration=false)
   &SubUVAnimation(
    "Sprite Renderer "=DataInterface() {propertyValues {SpriteRenderer=&self:NE_VC_Boost_Bulk_5.NiagaraSpriteRendererProperties_0}}
    SubUV Animation Mode:ENiagara_SubUVLookupModeV2=::NewEnumerator1
    SubUV Curve=DataInterface() {propertyValues {Curve {keys [
        {interpMode=RCIM_Cubic, tangentMode=RCTM_User, time=0, value=0.05}
        {arriveTangent=0.8947, interpMode=RCIM_Cubic, leaveTangent=0.8947, time=0.5, value=0.3}
        {arriveTangent=0.0754, interpMode=RCIM_Cubic, leaveTangent=0.0754, tangentMode=RCTM_User, time=0.95, value=0.9}
       ]}}}
    UseEndFrame=false
    UseStartFrame=true
   )
   &ScaleColor(
    Scale Alpha=&FloatFromCurve(
     FloatCurve=DataInterface() {propertyValues {Curve {
        defaultValue=1
        keys [{interpMode=RCIM_Cubic, tangentMode=RCTM_User, time=0, value=0}, {interpMode=RCIM_Cubic, tangentMode=RCTM_User, time=0.1, value=0.9685}]
       }}}
     Scale Curve=1.5
    )
    Scale RGB=&Multiply_Vector(
     A=&VectorFromFloat(Value=&MakeCustomFloatFromBool(
       Bool=$User.bIsPainted
       True Float=&FloatFromCurve(
        FloatCurve=DataInterface() {propertyValues {Curve {keys [
            <at 0 1.5 RCIM_Cubic>
            {arriveTangent=-1.0917, interpMode=RCIM_Cubic, leaveTangent=-1.0917, tangentMode=RCTM_User, time=0.12, value=0.925}
            {arriveTangent=-0.7005, interpMode=RCIM_Cubic, leaveTangent=-3.9584, tangentMode=RCTM_Break, time=0.48, value=0.8}
            {arriveTangent=-0.619, interpMode=RCIM_Cubic, leaveTangent=-0.619, tangentMode=RCTM_User, time=0.95, value=0}
           ]}}}
        Scale Curve=&MakeCustomFloatFromBool(Bool=$User.bIsPainted, True Float=4, False Float=1.35)
       )
       False Float=&FloatFromCurve(
        FloatCurve=DataInterface() {propertyValues {Curve {keys [
            <at 0 0.8 RCIM_Cubic>
            {arriveTangent=-1.0917, interpMode=RCIM_Cubic, leaveTangent=-1.0917, tangentMode=RCTM_User, time=0.12, value=0.925}
            {arriveTangent=-0.7005, interpMode=RCIM_Cubic, leaveTangent=-3.6875, tangentMode=RCTM_Break, time=0.3, value=0.7}
            {arriveTangent=-0.619, interpMode=RCIM_Cubic, leaveTangent=-0.619, tangentMode=RCTM_User, time=1, value=0}
           ]}}}
        Scale Curve=&MakeCustomFloatFromBool(Bool=$User.bIsPainted, True Float=6, False Float=1)
       )
      ))
     B=&VectorFromFloat(Value=&RandomRangeFloat(Minimum=0.75, Maximum=2))
    )
    ScaleA=true
    ScaleRGB=true
   ) {module=ScaleColor001}
  ]
  rendererClasses [&NiagaraSpriteRendererProperties]
  renderers [{rendererClass=&NiagaraSpriteRendererProperties}]
 }
 NE_VC_Boost_Pointed {
  emitterUpdateScript [
   &EmitterState(Life Cycle Mode=::NewEnumerator1, Loop Duration=0.1, UseLoopDelay=false)
   &VC_FilteredSocketCount()
   &SpawnRate(SpawnRate=60, Use Spawn Probability=false)
  ]
  particleSpawnScript [
   &VC_PaintedVariantsOverride(
    Black=<rgba 0.005 0.005 0.005 1>
    BurntSienna=<rgba 0.0747 0.0298 0.0136 1>
    Cobalt=<rgba 0.0621 0.1464 0.5677 1>
    Crimson=<rgba 0.4774 0 0 1>
    DefaultColor=<rgba 0.0142 0.65 0.4593 1>
    ForestGreen=<rgba 0.0148 0.2309 0.0257 1>
    Grey=<rgba 0.18 0.18 0.18 1>
    Orange=<rgba 0.3 0.085 0 1>
    Pink=<rgba 0.8108 0.1844 0.5915 1>
    Purple=<rgba 0.156 0.0175 0.35 1>
    Saffron=<rgba 0.5 0.4167 0 1>
    SkyBlue=<rgba 0 0.375 0.5 1>
    TitaniumWhite=<rgba 0.9 0.9 0.9 1>
   )
   &InitializeParticle(
    Color Mode=::NewEnumerator0
    Lifetime=&MakeCustomFloatFromBool(Bool=$User.bIsFrontendPreview, True Float=0.75, False Float=0.5)
    Sprite Size Max=<50 75>
    Sprite Size Min=<35 35>
    Sprite Size Mode=::NewEnumerator2
    UsePositionOffset=false
    Write Material Random=true
   )
   &PartitionParticles(Number Of Partitions=$Emitter.PortCount)
   &SkeletalMeshLocation(
    Bone / Socket Index=$Particles.PartitionParticles.Partition
    Bone / Socket Sampling Mode=::NewEnumerator5
    MaskSpawnGroup=false
    Mesh Sampling Type=::NewEnumerator6
    Sampled Position Offset=&AddVectorToPosition(Position=<0 0 0>, Vector=$User.DTCarOffset)
    Skeletal Mesh=$System.SkeletalMesh
   )
   &AddVelocity(
    Velocity=&MergeHelper_VectorSwitch(0=<1000 0 0>, 1=<1000 0 0>, 2=<100 0 0>, 3=?, Index=$User.GameModeID, Num=::NewEnumerator1)
    Velocity Speed Scale=-1
   )
   &ShapeLocation(
    MaskSpawnGroup=false
    Override Seed=false
    Sphere Radius=&MakeCustomFloatFromBool(Bool=$User.bIsFrontendPreview, True Float=15, False Float=0)
    UseOwnerScale=true
   )
   &CameraOffset(Camera Offset Amount=-50)
  ]
  particleUpdateScript [
   &ParticleState()
   &ScaleSpriteSize(
    Non-Uniform Curve Scale=<2 1.5>
    Non-Uniform Curve Sprite Scale=DataInterface() {propertyValues {
      XCurve {keys [<at 0 0.25 RCIM_Cubic>, {arriveTangent=1.8469, interpMode=RCIM_Cubic, leaveTangent=1.8469, tangentMode=RCTM_User, time=1, value=0.8}]}
      YCurve {keys [<at 0 0.4 RCIM_Cubic>, {arriveTangent=1.7282, interpMode=RCIM_Cubic, leaveTangent=1.7282, tangentMode=RCTM_User, time=1, value=1}]}
     }}
    Scale Sprite Size Mode=::NewEnumerator3
   )
   &SpriteRotationRate(Rotation Rate=10)
   &Drag(
    Drag=&MakeCustomFloatFromBool(Bool=$User.bIsFrontendPreview, True Float=0, False Float=2)
    Use Linear Drag=true
    Use Rotational Drag=true
   )
   &InheritSourceMovement(
    Applied Position Scale=&Multiply_VectorByFloat(
     Vector=<1 1 1>
     Float=&FloatFromCurve(FloatCurve=DataInterface() {propertyValues {Curve {keys [
          {arriveTangent=0.0056, interpMode=RCIM_Cubic, leaveTangent=0.0056, tangentMode=RCTM_User, time=0.1, value=1}
          {arriveTangent=-0.248, interpMode=RCIM_Cubic, leaveTangent=-0.248, tangentMode=RCTM_User, time=1, value=0.5}
         ]}}})
    )
    ApplyForce=false
    ApplyPosition=true
    ApplyPositionFalloff=false
   )
   &SolveForcesAndVelocity(Clamp Velocity=false, Limit Acceleration=false)
   &SubUVAnimation(
    End Frame Range Override:NiagaraInt32=&RandomRangeInt(Minimum=6, Maximum=8)
    "Sprite Renderer "=DataInterface() {propertyValues {SpriteRenderer=&self:NE_VC_Boost_Pointed_12.NiagaraSpriteRendererProperties_0}}
    SubUV Animation Mode:ENiagara_SubUVLookupModeV2=::NewEnumerator1
    SubUV Curve=DataInterface() {propertyValues {Curve {keys [{tangentMode=RCTM_User, time=0, value=0}, {tangentMode=RCTM_User, time=1, value=1}]}}}
    UseEndFrame=true
    UseStartFrame=true
   )
   &ScaleColor(
    Scale Alpha=&FloatFromCurve(FloatCurve=DataInterface() {propertyValues {Curve {
        defaultValue=1
        keys [{interpMode=RCIM_Cubic, tangentMode=RCTM_User, time=0.1, value=0}, {interpMode=RCIM_Cubic, tangentMode=RCTM_User, time=0.2, value=1}]
       }}})
    Scale RGB=&VectorFromFloat(Value=&FloatFromCurve(FloatCurve=DataInterface() {propertyValues {Curve {keys [<at 0 1 RCIM_Cubic>, {arriveTangent=-1.5141, interpMode=RCIM_Cubic, leaveTangent=-1.5141, tangentMode=RCTM_User, time=1, value=0.5}]}}}))
    ScaleA=true
    ScaleRGB=true
   )
  ]
  rendererClasses [&NiagaraSpriteRendererProperties]
  renderers [{rendererClass=&NiagaraSpriteRendererProperties}]
 }
 NE_VC_Boost_Rings001 {
  emitterUpdateScript [
   &EmitterState(Life Cycle Mode=::NewEnumerator1, Loop Duration=0.1, UseLoopDelay=false)
   &VC_FilteredSocketCount()
   &SpawnRate(SpawnRate=60, Use Spawn Probability=false)
  ]
  particleSpawnScript [
   &VC_PaintedVariantsOverride(
    Black=<rgba 0.01 0.01 0.01 1>
    BurntSienna=<rgba 0.1 0.034 0.01 1>
    Cobalt=<rgba 0.075 0.1458 0.5 1>
    Crimson=<rgba 0.2135 0 0 1>
    DefaultColor=<rgba 0 0.6 0.32 1>
    ForestGreen=<rgba 0 0.1372 0.0069 1>
    Grey=<rgba 0.1 0.1 0.1 1>
    Lime=<rgba 0.319 0.4253 0 1>
    Orange=<rgba 0.35 0.0992 0 1>
    Pink=<rgba 0.34 0.153 0.2777 1>
    Purple=<rgba 0.1783 0.02 0.4 1>
    Saffron=<rgba 0.5 0.4583 0 1>
    SkyBlue=<rgba 0.026 0.4087 0.6 1>
    TitaniumWhite=<rgba 0.3 0.3 0.3 1>
   )
   &InitializeParticle(
    Color Mode=::NewEnumerator0
    Lifetime=&MakeCustomFloatFromBool(Bool=$User.bIsFrontendPreview, True Float=0.9, False Float=0.5)
    Sprite Rotation Mode=::NewEnumerator2
    Sprite Size Mode=::NewEnumerator1
    Uniform Sprite Size Max=60
    Uniform Sprite Size Min=35
    UsePositionOffset=false
    Write Material Random=true
   )
   &PartitionParticles(Number Of Partitions=$Emitter.PortCount)
   &SkeletalMeshLocation(
    Bone / Socket Index=$Particles.PartitionParticles.Partition
    Bone / Socket Sampling Mode=::NewEnumerator5
    MaskSpawnGroup=false
    Mesh Sampling Type=::NewEnumerator6
    Sampled Position Offset=&AddVectorToPosition(Position=<-25 0 0>, Vector=$User.DTCarOffset)
    Skeletal Mesh=$System.SkeletalMesh
   )
   &AddVelocity(
    Velocity=&MergeHelper_VectorSwitch(0=<1000 0 0>, 1=<1000 0 0>, 2=<100 0 0>, 3=?, Index=$User.GameModeID, Num=::NewEnumerator1)
    Velocity Speed Scale=-1
   )
   &ShapeLocation(
    MaskSpawnGroup=false
    Override Seed=false
    Sphere Radius=&MakeCustomFloatFromBool(Bool=$User.bIsFrontendPreview, True Float=10, False Float=0)
    UseOwnerScale=true
   )
   &CameraOffset(Camera Offset Amount=-50)
  ]
  particleUpdateScript [
   &ParticleState()
   &ScaleSpriteSize(Uniform Curve Sprite Scale=DataInterface() {propertyValues {Curve {keys [
        {arriveTangent=1.2176, interpMode=RCIM_Cubic, leaveTangent=1.2176, tangentMode=RCTM_User, time=0, value=0.5}
        {arriveTangent=0.0224, interpMode=RCIM_Cubic, leaveTangent=0.0224, tangentMode=RCTM_User, time=1, value=1}
       ]}}})
   &SpriteRotationRate(Delta Time:NiagaraFloat=$Particles.Lifetime, Rotation Rate=&RandomRangeFloat(Minimum=10, Maximum=-10))
   &Drag(
    Drag=&MakeCustomFloatFromBool(Bool=$User.bIsFrontendPreview, True Float=0, False Float=2)
    Use Linear Drag=true
    Use Rotational Drag=true
   )
   &InheritSourceMovement(
    Applied Position Scale=&Multiply_VectorByFloat(
     Vector=<1 1 1>
     Float=&FloatFromCurve(FloatCurve=DataInterface() {propertyValues {Curve {keys [
          {arriveTangent=0.0056, interpMode=RCIM_Cubic, leaveTangent=0.0056, tangentMode=RCTM_User, time=0.1, value=1}
          {arriveTangent=-0.248, interpMode=RCIM_Cubic, leaveTangent=-0.248, tangentMode=RCTM_User, time=1, value=0.5}
         ]}}})
    )
    ApplyForce=false
    ApplyPosition=true
    ApplyPositionFalloff=false
   )
   &SolveForcesAndVelocity(Clamp Velocity=false, Limit Acceleration=false)
   &SubUVAnimation(
    End Frame Range Override=7
    "Sprite Renderer "=DataInterface() {propertyValues {SpriteRenderer=&self:NE_VC_Boost_Rings001_8.NiagaraSpriteRendererProperties_0}}
    SubUV Animation Mode:ENiagara_SubUVLookupModeV2=::NewEnumerator1
    SubUV Curve=DataInterface() {propertyValues {Curve {keys [
        {arriveTangent=1.6278, interpMode=RCIM_Cubic, leaveTangent=1.6278, tangentMode=RCTM_User, time=0, value=0}
        {interpMode=RCIM_Cubic, tangentMode=RCTM_User, time=1, value=0.9}
       ]}}}
    UseEndFrame=true
    UseStartFrame=true
   )
   &ScaleColor(
    Scale Alpha=&FloatFromCurve(FloatCurve=DataInterface() {propertyValues {Curve {keys [
         {tangentMode=RCTM_User, time=0, value=0}
         <at 0.0983 1>
         {arriveTangent=-0.0405, interpMode=RCIM_Cubic, leaveTangent=-0.0405, tangentMode=RCTM_User, time=0.8, value=1}
         {arriveTangent=-14.5622, interpMode=RCIM_Cubic, leaveTangent=-14.5622, tangentMode=RCTM_User, time=1, value=0}
        ]}}})
    Scale RGB=&VectorFromFloat(Value=&FloatFromCurve(
      FloatCurve=DataInterface() {propertyValues {Curve {keys [
          <at 0 1.25 RCIM_Cubic>
          {arriveTangent=-0.1024, interpMode=RCIM_Cubic, leaveTangent=-0.1024, tangentMode=RCTM_User, time=0.3, value=0.9}
          {arriveTangent=-0.0192, interpMode=RCIM_Cubic, leaveTangent=-0.0192, tangentMode=RCTM_User, time=1, value=1}
         ]}}}
      Scale Curve=1.2
     ))
    ScaleA=true
    ScaleRGB=true
   )
  ]
  rendererClasses [&NiagaraSpriteRendererProperties]
  renderers [{rendererClass=&NiagaraSpriteRendererProperties}]
 }
 NE_VC_Boost_Source001 {
  emitterUpdateScript [
   &EmitterState(Life Cycle Mode=::NewEnumerator1, Loop Duration=1, UseLoopDelay=false)
   &VC_FilteredSocketCount()
   &SpawnRate(SpawnRate=150, Use Spawn Probability=false)
  ]
  particleSpawnScript [
   &VC_PaintedVariantsOverride(
    Black=<rgba 0.2 0.2 0.2 1>
    BurntSienna=<rgba 0.4 0.3183 0.26 1>
    Cobalt=<rgba 0.1984 0.2929 0.5365 1>
    Crimson=<rgba 0.4774 0.1633 0.1842 1>
    DefaultColor=<rgba 0.014 0.6406 0.5884 1>
    ForestGreen=<rgba 0.2884 0.4288 0.2181 1>
    Grey=<rgba 0.3 0.3 0.3 1>
    Lime=<rgba 0.5853 0.7656 0.2246 1>
    Orange=<rgba 0.4392 0.2461 0.1426 1>
    Pink=<rgba 0.5712 0.3223 0.4882 1>
    Purple=<rgba 0.349 0.0241 0.8038 1>
    Saffron=<rgba 0.8432 0.8976 0.2447 1>
    SkyBlue=<rgba 0.0818 0.6951 0.8177 1>
    TitaniumWhite=<rgba 0.7031 0.7031 0.7031 1>
   )
   &InitializeParticle(
    Color Mode=::NewEnumerator0
    Lifetime Max=0.25
    Lifetime Min=0.1
    Lifetime Mode=::NewEnumerator1
    Sprite Rotation Mode=::NewEnumerator2
    Sprite Size=&MakeCustomVector2DFromBool(Bool=$User.bIsFrontendPreview, True Vector2D=<25 125>, False Vector2D=<25 100>)
    Sprite Size Mode=::NewEnumerator4
    UsePositionOffset=false
    Write Material Random=true
   )
   &PartitionParticles(Number Of Partitions=$Emitter.PortCount)
   &SkeletalMeshLocation(
    Bone / Socket Index=$Particles.PartitionParticles.Partition
    Bone / Socket Sampling Mode=::NewEnumerator5
    MaskSpawnGroup=false
    Mesh Sampling Type=::NewEnumerator6
    Sampled Position Offset=&AddVectorToPosition(Position=<-30 0 0>, Vector=$User.DTCarOffset)
    Skeletal Mesh=$System.SkeletalMesh
   )
   &AddVelocity(
    Cone Angle=&RandomRangeFloat(Minimum=50, Maximum=75)
    Cone Axis=<-1 0 0>
    Module.UseConeFalloff=false
    Module.UseVelDistribution=false
    Override Seed=false
    Velocity Mode=::NewEnumerator2
    Velocity Speed=&RandomRangeFloat(Minimum=500, Maximum=750)
   )
   &AddVelocity(
    Velocity=&MakeCustomVectorFromBool(Bool=$User.bIsFrontendPreview, True Vector=<500 0 0>, False Vector=<200 0 0>)
    Velocity Speed Scale=&RandomRangeFloat(Minimum=-0.5, Maximum=-1)
   ) {module=AddVelocity001}
   &ShapeLocation(MaskSpawnGroup=false, Override Seed=false, Sphere Radius=10, UseOwnerScale=true)
   &CameraOffset(Camera Offset Amount=-50)
  ]
  particleUpdateScript [
   &ParticleState()
   &ScaleSpriteSize(
    Uniform Curve Scale=&MakeCustomFloatFromBool(Bool=$User.bIsFrontendPreview, True Float=1.1, False Float=1.1)
    Uniform Curve Sprite Scale=DataInterface() {propertyValues {Curve {keys [
        {arriveTangent=4.3733, interpMode=RCIM_Cubic, leaveTangent=4.3733, tangentMode=RCTM_User, time=0, value=0}
        {arriveTangent=-0.0405, interpMode=RCIM_Cubic, leaveTangent=-0.0405, tangentMode=RCTM_User, time=0.15, value=1}
        {arriveTangent=-2.0583, interpMode=RCIM_Cubic, leaveTangent=-2.0583, tangentMode=RCTM_User, time=1, value=0}
       ]}}}
   )
   &WindForce(Frequency=10, Scale=2, Wind Speed=<1000 0 0>)
   &Drag(
    Drag=&MakeCustomFloatFromBool(Bool=$User.bIsFrontendPreview, True Float=0.5, False Float=1)
    Use Linear Drag=true
    Use Rotational Drag=true
   )
   &SolveForcesAndVelocity(Clamp Velocity=false, Limit Acceleration=false)
   &SubUVAnimation(
    End Frame Range Override=6
    "Sprite Renderer "=DataInterface() {propertyValues {SpriteRenderer=&self:NE_VC_Boost_Source001_0.NiagaraSpriteRendererProperties_0}}
    Start Frame Range Override=3
    SubUV Animation Mode:ENiagara_SubUVLookupModeV2=::NewEnumerator2
    UseEndFrame=true
    UseStartFrame=true
   )
   &ScaleColor(
    Scale Alpha=&FloatFromCurve(FloatCurve=DataInterface() {propertyValues {Curve {
        defaultValue=1
        keys [{interpMode=RCIM_Cubic, tangentMode=RCTM_User, time=0, value=0}, {interpMode=RCIM_Cubic, tangentMode=RCTM_User, time=0.5, value=1}]
       }}})
    ScaleA=true
    ScaleRGB=true
   )
  ]
  rendererClasses [&NiagaraSpriteRendererProperties]
  renderers [{rendererClass=&NiagaraSpriteRendererProperties}]
 }
 NE_VC_Boost_Tiny {
  emitterUpdateScript [
   &EmitterState(Life Cycle Mode=::NewEnumerator1, Loop Duration=1, UseLoopDelay=false)
   &VC_FilteredSocketCount()
   &SpawnRate(SpawnRate=50, Use Spawn Probability=false)
  ]
  particleSpawnScript [
   &VC_PaintedVariantsOverride(
    Black=<rgba 0.005 0.005 0.005 1>
    Cobalt=<rgba 0.0362 0.0837 0.3212 1>
    Crimson=<rgba 0.8 0 0.0217 1>
    DefaultColor=<rgba 0.0438 3 0.8589 1>
    ForestGreen=<rgba 0 0.1615 0.0082 1>
    Grey=<rgba 0.25 0.25 0.25 1>
    Lime=<rgba 0.3924 0.5885 0 1>
    Orange=<rgba 0.2865 0.1262 0.0403 1>
    Pink=<rgba 0.3976 0.159 0.3379 1>
    Purple=<rgba 0.2743 0 0.4115 1>
    Saffron=<rgba 0.9 0.825 0 1>
    SkyBlue=<rgba 0 0.2876 0.4323 1>
   )
   &InitializeParticle(
    Color Mode=::NewEnumerator0
    Lifetime=&MakeCustomFloatFromBool(Bool=$User.bIsFrontendPreview, True Float=0.8, False Float=0.5)
    Sprite Rotation Mode=::NewEnumerator2
    Sprite Size Mode=::NewEnumerator1
    Uniform Sprite Size Max=35
    Uniform Sprite Size Min=20
    UsePositionOffset=false
    Write Material Random=true
   )
   &PartitionParticles(Number Of Partitions=$Emitter.PortCount)
   &SkeletalMeshLocation(
    Bone / Socket Index=$Particles.PartitionParticles.Partition
    Bone / Socket Sampling Mode=::NewEnumerator5
    MaskSpawnGroup=false
    Mesh Sampling Type=::NewEnumerator6
    Sampled Position Offset=&AddVectorToPosition(Position=<0 0 0>, Vector=$User.DTCarOffset)
    Skeletal Mesh=$System.SkeletalMesh
   )
   &AddVelocity(
    Velocity=&MergeHelper_VectorSwitch(0=<1000 0 0>, 1=<1000 0 0>, 2=<100 0 0>, 3=?, Index=$User.GameModeID, Num=::NewEnumerator1)
    Velocity Speed Scale=-1
   )
   &ShapeLocation(
    MaskSpawnGroup=false
    Override Seed=false
    Sphere Radius=&MakeCustomFloatFromBool(Bool=$User.bIsFrontendPreview, True Float=10, False Float=0)
    UseOwnerScale=true
   )
   &CameraOffset(Camera Offset Amount=-50)
  ]
  particleUpdateScript [
   &ParticleState()
   &ScaleSpriteSize(
    Uniform Curve Scale=2
    Uniform Curve Sprite Scale=DataInterface() {propertyValues {Curve {keys [
        {arriveTangent=1.6969, interpMode=RCIM_Cubic, leaveTangent=1.6969, tangentMode=RCTM_User, time=0, value=0.25}
        {arriveTangent=0.0224, interpMode=RCIM_Cubic, leaveTangent=0.0224, tangentMode=RCTM_User, time=1, value=1}
       ]}}}
   )
   &SpriteRotationRate(Rotation Rate=&RandomRangeFloat(Minimum=-10, Maximum=10))
   &Drag(
    Drag=&MakeCustomFloatFromBool(Bool=$User.bIsFrontendPreview, True Float=0, False Float=2)
    Use Linear Drag=true
    Use Rotational Drag=true
   )
   &InheritSourceMovement(
    Applied Position Scale=&Multiply_VectorByFloat(
     Vector=<1 1 1>
     Float=&FloatFromCurve(FloatCurve=DataInterface() {propertyValues {Curve {keys [
          {arriveTangent=0.0056, interpMode=RCIM_Cubic, leaveTangent=0.0056, tangentMode=RCTM_User, time=0.1, value=1}
          {arriveTangent=-0.248, interpMode=RCIM_Cubic, leaveTangent=-0.248, tangentMode=RCTM_User, time=1, value=0.5}
         ]}}})
    )
    ApplyForce=false
    ApplyPosition=true
    ApplyPositionFalloff=false
   )
   &SolveForcesAndVelocity(Clamp Velocity=false, Limit Acceleration=false)
   &SubUVAnimation(
    End Frame Range Override=9
    "Sprite Renderer "=DataInterface() {propertyValues {SpriteRenderer=&self:NE_VC_Boost_Tiny_9.NiagaraSpriteRendererProperties_0}}
    Start Frame Range Override=6
    SubUV Animation Mode:ENiagara_SubUVLookupModeV2=::NewEnumerator1
    SubUV Curve=DataInterface() {propertyValues {Curve {keys [{interpMode=RCIM_Cubic, tangentMode=RCTM_User, time=0, value=0}, {interpMode=RCIM_Cubic, tangentMode=RCTM_User, time=1, value=1}]}}}
    UseEndFrame=true
    UseStartFrame=true
   )
   &ScaleColor(
    Scale Alpha=&FloatFromCurve(
     FloatCurve=DataInterface() {propertyValues {Curve {
        defaultValue=1
        keys [{interpMode=RCIM_Cubic, tangentMode=RCTM_User, time=0, value=0}, {interpMode=RCIM_Cubic, tangentMode=RCTM_User, time=0.2, value=1}]
       }}}
     Scale Curve=10
    )
    Scale RGB=&VectorFromFloat(Value=&FloatFromCurve(
      FloatCurve=DataInterface() {propertyValues {Curve {keys [{interpMode=RCIM_Cubic, tangentMode=RCTM_User, time=0, value=0.5}, {interpMode=RCIM_Cubic, tangentMode=RCTM_User, time=1, value=1}]}}}
      Scale Curve=2
     ))
    ScaleA=true
    ScaleRGB=true
   )
  ]
  rendererClasses [&NiagaraSpriteRendererProperties]
  renderers [{rendererClass=&NiagaraSpriteRendererProperties}]
 }
}
summary {
 systemName=NS_Boost_2D_Smoke
 emitters [
  {emitterName=NE_VC_Boost_Bulk, rendererClasses [&NiagaraSpriteRendererProperties]}
  {emitterName=NE_VC_Boost_Tiny, rendererClasses [&NiagaraSpriteRendererProperties]}
  {emitterName=NE_VC_Boost_Rings001, rendererClasses [&NiagaraSpriteRendererProperties]}
  {emitterName=NE_VC_Boost_Pointed, rendererClasses [&NiagaraSpriteRendererProperties]}
  {emitterName=NE_VC_Boost_Source001, rendererClasses [&NiagaraSpriteRendererProperties]}
 ]
 userVariables [
  User.bIsFrontendPreview {type=NiagaraBool, defaultValue=true}
  User.bIsPainted {type=NiagaraBool}
  User.ColorVariantID {type=NiagaraInt32}
  User.DTCarOffset {type=Vector3f, description="", defaultValue {struct=Vector3f, value=<0 0 0>}}
  User.GameModeID {type=NiagaraInt32}
  User.PaintedColor {type=LinearColor, description="", defaultValue {struct=LinearColor, value=<rgba 1 1 1 1>}}
  User.PaintedScalar {type=NiagaraFloat, defaultValue=1, description=""}
  User.PortCount {type=NiagaraInt32, defaultValue=2}
 ]
}
